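# Physics asset for the offhand axe mesh of the LycheeNickel pickaxe cosmetic, a Fortnite melee weapon.
# It holds a single capsule body on the Melee_TwoHanded_Handle bone, slightly offset from the bone and rotated.
# It defines no constraints, so the axe is represented physically as one simple shape.
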PhysicsAsset /BRCosmetics/Weapons/FORT_Melee/Pickaxe_LycheeNickel/Meshes/Offhand/LycheeNickel_Offhand_Axe_PhysicsAsset
bodies {Melee_TwoHanded_Handle [{shapeType=Capsule, length=13.3477, radius=13.3477, center=<3.9914 -0.4054 -0.9878>, rotation=<rot 90 -90 0>}]}
constraints []
